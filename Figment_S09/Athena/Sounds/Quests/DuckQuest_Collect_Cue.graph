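# Sound cue for collecting in the Figment Season 9 duck quest, playing the DuckQuest_Collect wave.
# A short fixed Delay feeds a Modulator at default settings, which adds slight random pitch and volume variation to the single Wave Player.
# It routes through the Fort_UI sound class with a small volume boost, yet is spatialized by the shared quest-interactable hint attenuation.
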SoundCue /Figment_S09/Athena/Sounds/Quests/DuckQuest_Collect_Cue
root=Delay_0
nodes [
 {id=Delay_0, children [Modulator_0], properties {delayMax=0.25, delayMin=0.25}}
 {id=Modulator_0, children [WavePlayer_1]}
 {id=WavePlayer_1, properties {soundWaveAssetPtr=&/DuckQuest_Collect}}
]
properties {
 attenuationSettings=/Game/Sounds/Attenuation/Fort_QuestInteractables_Hint
 maxDistance=1300
 soundClassObject=&Fort_UI
 volumeMultiplier=1.1
}
