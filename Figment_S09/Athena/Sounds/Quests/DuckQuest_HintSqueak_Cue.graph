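# DuckQuest_HintSqueak_Cue is a sound cue that plays a hint squeak for the Duck Quest.
# A random node picks one of four equally weighted squeak waves, and a modulator above it varies each playback.
# It uses the quest interactable hint attenuation and the UI sound class, with a limited range and slightly reduced volume.
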SoundCue /Figment_S09/Athena/Sounds/Quests/DuckQuest_HintSqueak_Cue
root=Modulator_0
nodes [
 {id=Modulator_0, children [Random_0]}
 {id=Random_0, children [WavePlayer_0,WavePlayer_2,WavePlayer_3,WavePlayer_4], properties {weights [1,1,1,1]}}
 {id=WavePlayer_0, properties {soundWaveAssetPtr=&/DuckQuest_HintSqueak_01}}
 {id=WavePlayer_2, properties {soundWaveAssetPtr=&/DuckQuest_HintSqueak_02}}
 {id=WavePlayer_3, properties {soundWaveAssetPtr=&/DuckQuest_HintSqueak_03}}
 {id=WavePlayer_4, properties {soundWaveAssetPtr=&/DuckQuest_HintSqueak_04}}
]
properties {
 attenuationSettings=/Game/Sounds/Attenuation/Fort_QuestInteractables_Hint
 maxDistance=1300
 soundClassObject=&Fort_UI
 volumeMultiplier=0.8
}
