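# Sound cue that plays the interact sound for a duck quest, going by its name, in the Figment_S09 Athena content.
# A single wave player feeds a Modulator at default settings, which adds slight random pitch and volume variation per play.
# It uses the UI sound class with a quest-interactable hint attenuation asset and a slight volume boost.
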SoundCue /Figment_S09/Athena/Sounds/Quests/DuckQuest_Interact_Cue
root=Modulator_0
nodes [
 {id=Modulator_0, children [WavePlayer_1]}
 {id=WavePlayer_1, properties {soundWaveAssetPtr=&/DuckQuest_Interact}}
]
properties {
 attenuationSettings=/Game/Sounds/Attenuation/Fort_QuestInteractables_Hint
 maxDistance=1300
 soundClassObject=&Fort_UI
 volumeMultiplier=1.1
}
